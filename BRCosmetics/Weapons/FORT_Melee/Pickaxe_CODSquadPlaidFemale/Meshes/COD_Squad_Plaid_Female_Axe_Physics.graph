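# Physics asset for the COD Squad Plaid Female pickaxe, covering its dynamic spur bones.
# A sphere body on dyn_spur_01 and a capsule body on dyn_spur_base are joined by one constraint.
# The constraint frees swing1 and locks swing2 and twist, forming a single-axis hinge that lets the spur rotate relative to its base.
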PhysicsAsset /BRCosmetics/Weapons/FORT_Melee/Pickaxe_CODSquadPlaidFemale/Meshes/COD_Squad_Plaid_Female_Axe_Physics
bodies {dyn_spur_01 [{shapeType=Sphere, radius=8, center=<2 1 0>}], dyn_spur_base [{shapeType=Capsule, length=3, radius=3, rotation=<rot 90 0 0>}]}
constraints [{
  bone1Name=dyn_spur_01
  bone2Name=dyn_spur_base
  swing1LimitDegrees=180
  swing1Motion=Free
  swing2LimitDegrees=1
  swing2Motion=Locked
  twistLimitDegrees=1
  twistMotion=Locked
 }]
