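# Animation Blueprint for the COD Squad Plaid Female pickaxe that adds physics-driven secondary motion to the weapon mesh.
# The AnimGraph passes the input pose through a RigidBody node using the axe's physics asset, simulated relative to b_AR_Root.
# Custom gravity, scaled component motion and clamped acceleration shape that movement.
# OnWeaponChanged checks bIsPlayingMeleeAnim and resets dynamics via teleport and reset modes; the update event is empty.
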AnimBlueprint /BRCosmetics/Weapons/FORT_Melee/Pickaxe_CODSquadPlaidFemale/Meshes/COD_Squad_Plaid_Female_Axe_AnimBP
target_skeleton=&/COD_Squad_Plaid_Female_Axe_Skeleton
graphs [
 AnimGraph {nodes [
   AnimGraphNode_SubInput_0 = Input Pose()
   n1 = RigidBody(ComponentPose=Local To Component(LocalPose=@AnimGraphNode_SubInput_0)) {
    bClampLinearTranslationLimitToRefPose=true
    bOverrideWorldGravity=true
    overridePhysicsAsset=&/COD_Squad_Plaid_Female_Axe_Physics
    simulationSpace=BaseBoneSpace
    alphaBoolBlend {blendOption=Linear}
    baseBoneRef {boneName=b_AR_Root}
    componentAppliedLinearAccClamp=<0 1000 1000>
    componentLinearAccScale {y=0.3, z=0.3}
    componentLinearVelScale {y=0.1, z=0.1}
    overrideWorldGravity {z=-300}
   }
   Output Pose(Result=Component To Local(ComponentPose=@n1)) {layerGroup=None}
  ]}
 EventGraph {nodes [
   Event OnWeaponChanged()
   > Branch(Condition=Get bIsPlayingMeleeAnim())
   > ResetDynamics(InTeleportType=TeleportPhysics)
   > ResetDynamics(InTeleportType=ResetPhysics)
   TryGetPawnOwner()
   Event BlueprintUpdateAnimation()
  ]}
]
variables []
